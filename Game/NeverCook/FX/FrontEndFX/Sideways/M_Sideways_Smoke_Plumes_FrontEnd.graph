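# Translucent unlit material for front-end "Sideways" smoke plumes, using SplineThicken for camera-facing width that varies by vertex color and camera distance.
# Two panned, rotated smoke textures, one distorted, drive opacity with UV masks, distance fades and optional fresnel.
# A terrain height map optionally fades and collapses plumes by actor height.
# Color blends gradients, time-of-day tints and Schlick-phase fake sunlight, while low quality and simpler shading paths drop layers and wobble.
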Material /Game/NeverCook/FX/FrontEndFX/Sideways/M_Sideways_Smoke_Plumes_FrontEnd
graph {
 properties {blendMode=BLEND_Translucent, pixelDepthOffsetMode=PDOM_Legacy, shadingModel=MSM_Unlit, usedShadingModels=Unlit}
 nodes [
  Constant() {r=1}
  n1 = ComponentMask(None=TextureCoordinate()) {g=true}
  n2 = ScalarParameter() {defaultValue=2, parameterName=MaskExponent}
  n3 = ScalarParameter() {defaultValue=5000, parameterName=DepthFadeDistance}
  n4 = AppendVector(A=Constant() {r=1}, B=ScalarParameter() {defaultValue=0.5, group=Texture, parameterName=TilingSmoke_V_Mult_Tex2})
  n5 = TextureCoordinate() {vTiling=0.35}
  n6 = Multiply(A=ComponentMask(None=@n5) {g=true}, B=LinearInterpolate(Alpha=VertexColor().R) {constA=1.25, constB=2})
  n7 = Frac(None=ComponentMask(None=ObjectPositionWS()) {r=true})
  n8 = Add(A=AppendVector(A=ComponentMask(None=@n5) {r=true}, B=Add(A=@n6, B=@n7)), B=AppendVector(A=Constant(), B=VertexColor().R))
  n9 = VectorParameter() {group=Texture, parameterName=PannerSpeed_02, defaultValue=<rgba 0 0.01 0 1>}
  n10 = MaterialFunctionCall(
   UVs=Panner(Coordinate=Multiply(A=@n8, B=@n4), Speed=ComponentMask(None=@n9.RGB) {g=true, r=true})
   "Rotation Angle (0-1)"=ScalarParameter() {defaultValue=0.25, group=Texture, parameterName="Rotation Angle (0-1) (S)"}
  ) {materialFunction=&CustomRotator}
  n11 = Multiply(A=AppendVector(
    A=ScalarParameter() {defaultValue=0.5, group=Texture, parameterName=TilingDistort_X_Mult}
    B=ScalarParameter() {defaultValue=0.5, group=Texture, parameterName=TilingDistort_Y_Mult}
   ))
  n12 = ComponentMask(None=TextureSampleParameter2D(UVs=Multiply(A=@n10.Rotated Values, B=@n11)) {parameterName=DistortTexture_01, texture=/Game/Effects/Fort_Effects/Textures/Smoke/T_Puffy_Smoke_Trail_01}.RGB) {g=true, r=true}
  n13 = ScalarParameter() {parameterName=DistortAmount_01}
  n14 = Constant() {r=1}
  n15 = ScalarParameter() {defaultValue=0.5, group=Texture, parameterName=TilingSmoke_V_Mult}
  n16 = Panner(
   Coordinate=Multiply(A=AppendVector(A=@n14, B=@n15), B=@n8)
   Speed=ComponentMask(None=VectorParameter() {group=Texture, parameterName=PannerSpeed, defaultValue=<rgba 0 0.035 0 1>}.RGB) {g=true, r=true}
  )
  n17 = ScalarParameter() {defaultValue=0.25, group=Texture, parameterName="Rotation Angle (0-1) (S)"}
  n18 = TextureSampleParameter2D(UVs=Add(
    A=Multiply(A=@n12, B=@n13)
    B=MaterialFunctionCall(UVs=@n16, "Rotation Angle (0-1)"=@n17) {materialFunction=&CustomRotator}.Rotated Values
   )) {parameterName=SmokeTexture_01, texture=&T_Puffy_Smoke_Trail_01}
  n19 = TextureSampleParameter2D(UVs=@n10.Rotated Values) {parameterName=SmokeTexture_02, texture=&T_Puffy_Smoke_Trail_01}
  n20 = Multiply(A=@n18.G, B=ScalarParameter() {defaultValue=0.35, group=Texture, parameterName=OpacityLowEnd})
  n21 = QualitySwitch(Default=Multiply(A=Add(A=@n18.G, B=@n19.G), B=ScalarParameter() {defaultValue=0.25, group=Texture, parameterName=Opacity}), Low=@n20)
  n22 = ShadingPathSwitch(
   Default=ScalarParameter() {defaultValue=0.65, group=CloseFarFader, parameterName=DistanceScaleMult_LowQuality}
   Deferred=QualitySwitch(Default=Constant() {r=1}, Low=ScalarParameter() {defaultValue=0.65, group=CloseFarFader, parameterName=DistanceScaleMult_LowQuality})
  ) {desc=Squish WPO earlier on lower quality}
  n23 = Multiply(A=ScalarParameter() {defaultValue=25000, group=CloseFarFader, parameterName="FarFadeLength (S)"}, B=@n22)
  n24 = Multiply(A=ScalarParameter() {defaultValue=125000, group=CloseFarFader, parameterName="FarFadeOffset (S)"}, B=@n22)
  n25 = MaterialFunctionCall!(
   NearFadeLength=ScalarParameter() {defaultValue=21000, group=Scaler, parameterName="NearFadeLength (S)"}
   NearFadeOffset=ScalarParameter() {defaultValue=20000, group=Scaler, parameterName="NearFadeOffset (S)"}
   NearFadeCurve
   NearFadeStartingScale
   NearFadeEndingScale
   FarFadeLength=@n23
   FarFadeOffset=@n24
   FarFadeCurve
   FarFadeEndingScale
   Opacity Near Distance Bias=ScalarParameter() {group=Scaler, parameterName="Opacity Near Distance Bias Opacity (S)"}
   Opacity Far Distance Bias=ScalarParameter() {defaultValue=-2, group=Scaler, parameterName="Opacity Far Distance Bias (S)"}
  ) {materialFunction=/Game/Effects/Fort_Effects/Materials/_Functions/MF_SmoothLod}
  n26 = TextureCoordinate()
  n27 = ComponentMask(None=@n26) {g=true}
  n28 = ComponentMask(None=@n26) {g=true}
  n29 = Multiply(
   A=Multiply(A=ShadingPathSwitch(Default=@n21, Forward=@n20, Mobile=@n20), B=Power(Base=@n25.Opacity))
   B=Multiply(A=Multiply(A=@n27, B=@n27), B=OneMinus(None=@n28))
  )
  n30 = QualitySwitch(Default=Multiply(A=DepthFade(FadeDistance=@n3), B=@n29), Low=@n29)
  n31 = ShadingPathSwitch(Default=@n30, Deferred=@n30, Forward=@n29)
  n32 = TextureSampleParameter2D(UVs=TextureCoordinate() {coordinateIndex=1}) {
   group=LavaHeightCheck
   parameterName=TerrainHeightMap
   samplerType=SAMPLERTYPE_LinearGrayscale
   texture=/Game/Environments/Sets/Vulcano/Textures/T_TerrainHeightMap
  }
  n33 = Subtract(
   A=LinearInterpolate(
    A=ScalarParameter() {defaultValue=-2500, group=LavaHeightCheck, parameterName=TerrainHeight Mininum}
    B=ScalarParameter() {defaultValue=11082.7412, group=LavaHeightCheck, parameterName=TerrainHeightMax}
    Alpha=@n32.R
   )
   B=ComponentMask(None=ActorPositionWS()) {b=true}
  )
  n34 = ScalarParameter() {defaultValue=1500, group=LavaHeightCheck, parameterName=TerrainHeightCheckRange}
  n35 = Clamp(None=OneMinus(None=Divide(A=@n33, B=@n34)))
  n36 = StaticSwitchParameter(True=Multiply(A=@n31, B=VertexInterpolator(VS=@n35).PS), False=@n31) {defaultValue=true, parameterName=UseHeightMapTexture?}
  n37 = ScalarParameter() {defaultValue=1500, group=WPO, parameterName=WidthBase_Min}
  n38 = ScalarParameter() {defaultValue=2500, group=WPO, parameterName=WidthBase_Max}
  n39 = VertexColor()
  n40 = MaterialFunctionCall(
   NearFadeLength=Divide(A=ScalarParameter() {defaultValue=21000, group=CloseFarFader, parameterName="NearFadeLength (S)"}, B=@n22)
   NearFadeOffset=Divide(A=ScalarParameter() {defaultValue=20000, group=CloseFarFader, parameterName="NearFadeOffset (S)"}, B=@n22)
   FarFadeLength=@n23
   FarFadeOffset=@n24
   Opacity Far Distance Bias=ScalarParameter() {defaultValue=-2, group=CloseFarFader, parameterName="Opacity Far Distance Bias (S)"}
  ) {materialFunction=&MF_SmoothLod}
  n41 = ScalarParameter() {defaultValue=18000, group=WPO, parameterName=WidthTip_Min}
  n42 = ScalarParameter() {defaultValue=35000, group=WPO, parameterName=WidthTip_Max}
  n43 = VertexColor()
  n44 = MaterialFunctionCall(
   WidthBase=Multiply(A=LinearInterpolate(A=@n37, B=@n38, Alpha=@n39.R), B=@n40.Opacity)
   WidthTip=Multiply(A=LinearInterpolate(A=@n41, B=@n42, Alpha=@n43.R), B=@n40.Opacity)
  ) {materialFunction=&SplineThicken}
  n45 = DotProduct(A=CameraVectorWS(), B=@n44.Normal)
  n46 = Power(Base=Saturate(None=Abs(None=@n45)), Exp=ScalarParameter() {defaultValue=5, parameterName=FresnelExponent})
  Multiply(A=OneMinus(None=Power(Base=@n1, Exp=@n2)), B=StaticSwitchParameter(True=Multiply(A=@n36, B=@n46), False=@n36) {parameterName=UseFresnel?})
  n47 = Constant3Vector()
  n48 = VertexNormalWS()
  n49 = Constant3Vector() {desc=Disable WPO anim in Low, constant=<rgba 0 0 20 0>}
  n50 = VectorParameter() {group=WPO, parameterName=TextureWPO_UVtiling, defaultValue=<rgba 1 2 0 0>}
  n51 = Panner(Coordinate=Multiply(A=TextureCoordinate(), B=ComponentMask(None=@n50.RGB) {g=true, r=true})) {speedY=0.15}
  n52 = TextureSample(UVs=@n51) {samplerType=SAMPLERTYPE_Normal, texture=&T_Weap_Distort_N_02}
  n53 = AppendVector(
   A=Multiply(A=ComponentMask(None=@n52.RGB) {r=true}, B=ScalarParameter() {defaultValue=25, group=WPO, parameterName=WPO_HeightX}) {constB=0.5}
   B=Multiply(A=ComponentMask(None=@n52.RGB) {g=true}, B=ScalarParameter() {group=WPO, parameterName=WPO_HeightY}) {constB=0.5}
  )
  n54 = Multiply(
   A=TextureSample(UVs=@n51) {texture=&T_Smooth_tile}.G
   B=ScalarParameter() {group=WPO, parameterName=WPO_HeightZ}
  ) {constB=0.6}
  n55 = MaterialFunctionCall() {materialFunction=&LinearGradient}
  n56 = Constant() {r=0.5}
  n57 = Transform(None=LinearInterpolate(A=Constant3Vector(), B=AppendVector(A=@n53, B=@n54), Alpha=SphereMask(A=@n55.VGradient, B=@n56, Radius=@n56) {hardnessPercent=0}))
  n58 = LinearInterpolate(
   A=ScalarParameter() {group=WPO, parameterName=WPO_Wave_Intensity_Close}
   B=ScalarParameter() {defaultValue=100, group=WPO, parameterName=WPO_Wave_Intenstiy}
   Alpha=Clamp(None=@n40.Opacity)
  )
  n59 = ShadingPathSwitch(Default=@n49, Deferred=QualitySwitch(Default=Multiply(A=@n57, B=@n58) {constB=80}, Low=@n49))
  n60 = ShadingPathSwitch(Default=@n44.WorldPositionOffset, Deferred=QualitySwitch(Default=Add(A=@n44.WorldPositionOffset, B=@n59), Low=@n44.WorldPositionOffset))
  StaticSwitchParameter(True=LinearInterpolate(A=Multiply(A=@n47, B=@n48), B=@n60, Alpha=@n35), False=@n60) {defaultValue=true, parameterName=UseHeightMapTexture?}
  n61 = CollectionParameter() {
   collection=&FortniteMaterialParameters
   parameterName=SunlightDirectionalVector
  }
  MaterialExpressionDotProduct_1 = DotProduct(A=ComponentMask(None=@n61) {b=true, g=true, r=true}, B=CameraVectorWS())
  MaterialExpressionScalarParameter_5 = ScalarParameter() {defaultValue=0.4, parameterName=Anisotropy}
  n62 = Custom!(CosTheta=@MaterialExpressionDotProduct_1, k=@MaterialExpressionScalarParameter_5) {
   code="\tfloat Inner = (1 + k * CosTheta);\r\n\treturn (1 - k * k) / (4 * PI * Inner * Inner);\r\n\r\n"
   description=Schlick Phase
   inputs [
    {
     inputName=CosTheta
     input {expression=&self:MaterialExpressionDotProduct_1}
    }
    {inputName=k, input {expression=&self:MaterialExpressionScalarParameter_5}}
   ]
  }
  n63 = ScalarParameter() {defaultValue=25, parameterName=ScaleCloudColor}
  n64 = LinearInterpolate(A=ScalarParameter() {parameterName=HueShiftLow}, B=ScalarParameter() {defaultValue=0.15, parameterName=HueShiftHigh}, Alpha=VertexColor().R)
  n65 = ComponentMask(None=TextureCoordinate()) {r=true}
  n66 = ComponentMask(None=TextureCoordinate()) {g=true}
  n67 = Multiply(
   A=LinearInterpolate(
    A=VectorParameter() {group=Color, parameterName=Color_Bottom_Base, defaultValue=<rgba 10 1.3943 0 0>}.RGB
    B=VectorParameter() {group=Color, parameterName=Color_Bottom, defaultValue=<rgba 3 0.4183 0 0>}.RGB
    Alpha=Multiply(A=@n65, B=@n66)
   )
   B=MaterialFunctionCall(
    Value_Morning=ScalarParameter() {defaultValue=1, group=Color, parameterName="Value_Morning (S)"}
    Value_Day=ScalarParameter() {defaultValue=2, group=Color, parameterName="Value_Day (S)"}
    Value_Evening=ScalarParameter() {defaultValue=1, group=Color, parameterName="Value_Evening (S)"}
    Value_Night=ScalarParameter() {defaultValue=1.5, group=Color, parameterName="Value_Night (S)"}
   ) {materialFunction=&Scalar_Time_of_Day}.Result
  )
  n68 = QualitySwitch(
   Default=MaterialFunctionCall(Hue Shift Percentage=@n64, Texture=@n67) {materialFunction=&HueShift}.Result
   Low=@n67
  )
  n69 = CameraVectorWS()
  n70 = Divide(
   A=Add(A=ComponentMask(None=@n69) {b=true}, B=ScalarParameter() {defaultValue=0.1, parameterName=offset})
   B=ScalarParameter() {defaultValue=0.2, parameterName=Div}
  )
  n71 = Multiply(
   A=TextureCoordinate()
   B=ComponentMask(None=VectorParameter() {group=Color, parameterName=TileColorMask, defaultValue=<rgba 1 0.35 5 1>}.RGB) {g=true, r=true}
  )
  n72 = LinearInterpolate(
   A=VectorParameter() {group=Color, parameterName=Color_Top, defaultValue=<rgba 2 2.75 5 1>}.RGB
   B=ShadingPathSwitch(Default=ShadingPathSwitch(Default=@n68, Mobile=@n67) {desc=Migrated from FeatureLevelSwitch}, Forward=@n67)
   Alpha=StaticSwitchParameter(True=Clamp(None=@n70), False=ComponentMask(None=@n71) {g=true}) {defaultValue=true, parameterName=UseHeightInsteadofUVtofade}
  )
  n73 = MaterialFunctionCall(
   Color_Morning=VectorParameter() {group=Color, parameterName="Color_Morning (V3)", defaultValue=<rgba 1.25 0.9257 0.708 0>}.RGB
   Color_Day=VectorParameter() {group=Color, parameterName="Color_Day (V3)", defaultValue=<rgba 1.25 0.5722 0.5598 0>}.RGB
   Color_Evening=VectorParameter() {group=Color, parameterName="Color_Evening (V3)", defaultValue=<rgba 1 0.6005 0.2592 0>}.RGB
   Color_Night=VectorParameter() {group=Color, parameterName="Color_Night (V3)", defaultValue=<rgba 1 0.4906 0.3 0>}.RGB
  ) {materialFunction=&Color_Time_of_Day}
  n74 = Multiply(A=Multiply(A=@n73.Result), B=@n73.Result)
  n75 = TextureCoordinate()
  n76 = LinearInterpolate(
   A=VectorParameter() {group=Color, parameterName=Color_Whole_Base, defaultValue=<rgba 2 2.75 5 1>}.RGB
   B=VectorParameter() {group=Color, parameterName=Color_Whole, defaultValue=<rgba 2 2.75 5 1>}.RGB
   Alpha=Power(Base=ComponentMask(None=@n75) {g=true}, Exp=ScalarParameter() {defaultValue=2, parameterName=ColorMaskExponent})
  )
  n77 = ShadingPathSwitch(
   Default=QualitySwitch(
    Default=MaterialFunctionCall(Hue Shift Percentage=@n64, Texture=@n76) {materialFunction=&HueShift}.Result
    Low=@n76
   )
   Mobile=@n76
  ) {desc=Migrated from FeatureLevelSwitch}
  n78 = Multiply(
   A=StaticSwitchParameter(True=Multiply(A=@n72, B=@n74), False=ShadingPathSwitch(Default=@n77, Forward=@n76)) {defaultValue=true, parameterName=UseTOD?}
   B=@n66
  )
  StaticSwitchParameter(True=Multiply(A=Multiply(A=@n62, B=@n63), B=@n78), False=@n78) {defaultValue=true, parameterName=UseFakeLightAndShadow?}
 ]
}
